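# Physics asset for the Wizard male medium Commando body cosmetic, providing capsule bodies and joints for its dynamic coat and a pipe.
# Mirrored front, front-mid, outer, back-mid and back chains hang from spine_03, the back pair branching from one center bone.
# Coat joints limit swing and mostly keep twist very tight, with asymmetric front tuning.
# Pelvis, spine and thigh capsules form the body volume with free angular joints, while the pipe attaches to the pelvis.
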PhysicsAsset /BRCosmetics/Characters/Player/Male/Medium/Bodies/M_MED_Commando_Wizard/Meshes/M_MED_Wizard_Dyn_Coat
bodies {
 dyn_coat_bk_02 [{shapeType=Capsule, length=11, radius=1.5, center=<4.5296 0.3053 2.0654>, rotation=<rot 0 0 90>}]
 dyn_coat_bk_02_l [{shapeType=Capsule, length=9.5, radius=3.5, center=<11.076 1.7132 0.5542>, rotation=<rot 90 0 0>}]
 dyn_coat_bk_02_r [{shapeType=Capsule, length=9.5, radius=3.5, center=<-10.7907 -1.6536 -0.1016>, rotation=<rot 90 0 0>}]
 dyn_coat_bk_03_l [{shapeType=Capsule, length=11, radius=4.25, center=<7.9918 0.3051 0.0002>, rotation=<rot 0 0 90>}]
 dyn_coat_bk_03_r [{shapeType=Capsule, length=11, radius=4.25, center=<-3.9074 0.3051 0.0002>, rotation=<rot 0 0 90>}]
 dyn_coat_bk_04_l [{shapeType=Capsule, length=9, radius=3, center=<4.3853 3.3755 0>, rotation=<rot -0.649 1.5259 99.8534>}]
 dyn_coat_bk_04_r [{shapeType=Capsule, length=9, radius=3, center=<-3.9442 -3.6195 0.7143>, rotation=<rot -0.2917 0.7769 109.9287>}]
 dyn_coat_bk_mid_01_l [{shapeType=Capsule, length=10.5938, radius=3, center=<12.2027 1.8987 0.0001>, rotation=<rot 0 0 90>}]
 dyn_coat_bk_mid_01_r [{
   shapeType=Capsule
   length=6.536
   radius=4.105
   shapeName=dyn_coat_bk_mid_01_r_capsule
   center=<-12.505 -1.3655 0.7652>
   rotation=<rot 3.952 178.2811 -89.6535>
  }]
 dyn_coat_bk_mid_02_l [{shapeType=Capsule, length=11, radius=3, center=<11.3206 3.8905 1.4008>, rotation=<rot 0 0 90>}]
 dyn_coat_bk_mid_02_r [{shapeType=Capsule, length=11, radius=3, center=<-11.5042 -4.0624 -0.7542>, rotation=<rot 0 0 90>}]
 dyn_coat_bk_mid_03_l [{shapeType=Capsule, length=13.2188, radius=3, center=<10.2517 3.8906 -0.0001>, rotation=<rot 0 0 90>}]
 dyn_coat_bk_mid_03_r [{shapeType=Capsule, length=13, radius=3, center=<-9.8845 -4.0623 -0.0001>, rotation=<rot 0 0 90>}]
 dyn_coat_bk_mid_04_l [{shapeType=Capsule, length=11, radius=5, center=<6.6431 3.8905 0>, rotation=<rot 0 0 90>}]
 dyn_coat_bk_mid_04_r [{shapeType=Capsule, length=15, radius=5, center=<-4.6194 -3.0518 0>, rotation=<rot 0 -0.001 90>}]
 dyn_coat_fr_01_l [{shapeType=Capsule, length=13.5312, radius=1.5, center=<7.5749 0.3048 -0.0358>, rotation=<rot 90 0.0198 0.0198>}]
 dyn_coat_fr_01_r [{shapeType=Capsule, length=10.5312, radius=1.5, center=<-5.7476 0.3051 -0.3632>, rotation=<rot 90 -0.0099 -0.0099>}]
 dyn_coat_fr_02_l [{shapeType=Capsule, length=7.8438, radius=3.5, center=<8.0062 2.3569 -0.7657>, rotation=<rot 0.0001 90.0013 89.9999>}]
 dyn_coat_fr_02_r [{shapeType=Capsule, length=24.7812, radius=4, center=<-9.1421 0.0001 0>, rotation=<rot 0 0 90>}]
 dyn_coat_fr_03_l [{shapeType=Capsule, length=14, radius=3, shapeName=dyn_coat_fr_03_l_capsule, center=<9.5944 0.3087 -0.5851>, rotation=<rot 0 0 90>}]
 dyn_coat_fr_03_r [{shapeType=Capsule, length=26.5, radius=4.5, shapeName=dyn_coat_fr_03_r_capsule, center=<-3.2365 0 -0.7883>, rotation=<rot 0 0 79.9995>}]
 dyn_coat_fr_04_l [{shapeType=Capsule, length=15, radius=3, shapeName=dyn_coat_fr_04_l_capsule, center=<13.3944 0.4466 -0.8462>, rotation=<rot 0 0 90>}]
 dyn_coat_fr_04_r [{
   shapeType=Capsule
   length=29
   radius=4.8
   shapeName=dyn_coat_fr_04_r_capsule
   center=<-2.4831 0.4114 0.1498>
   rotation=<rot -0.1603 -0.4408 69.9977>
  }]
 dyn_coat_fr_05_l [{shapeType=Capsule, length=15.25, radius=3.75, shapeName=dyn_coat_fr_05_l_capsule, center=<12.9058 0.2366 -0.4486>, rotation=<rot 0 0 90>}]
 dyn_coat_fr_05_r [{
   shapeType=Capsule
   length=29.25
   radius=5.2
   shapeName=dyn_coat_fr_05_r_capsule
   center=<-3.7254 0.6543 0.0572>
   rotation=<rot 0.4221 -5.1329 75.0549>
  }]
 dyn_coat_fr_06_l [{shapeType=Capsule, length=22, radius=5, shapeName=dyn_coat_fr_06_l_capsule, center=<5.6628 0 -0.0002>, rotation=<rot 0 0 90>}]
 dyn_coat_fr_06_r [{
   shapeType=Capsule
   length=34.25
   radius=5
   shapeName=dyn_coat_fr_06_r_capsule
   center=<-1.1526 0.0716 0.1236>
   rotation=<rot 6.1081 176.4405 -75.4005>
  }]
 dyn_coat_fr_mid_01_l [{shapeType=Capsule, length=17, radius=1.5, shapeName=dyn_coat_fr_mid_01_l_capsule, center=<16.6704 0.0001 -0.0001>, rotation=<rot 0 0 90>}]
 dyn_coat_fr_mid_01_r [{shapeType=Capsule, length=9.5, radius=3, shapeName=dyn_coat_fr_mid_01_r_capsule, center=<-15.2757 0 -0.0001>, rotation=<rot 0 0 90>}]
 dyn_coat_fr_mid_02_l [{shapeType=Capsule, length=17, radius=3, shapeName=dyn_coat_fr_mid_02_l_capsule, center=<5.6628 0 -0.0002>, rotation=<rot 0 0 90>}]
 dyn_coat_fr_mid_02_r [{shapeType=Capsule, length=15.25, radius=4.75, shapeName=dyn_coat_fr_mid_02_r_capsule, center=<-5.6627 1.0523 0>, rotation=<rot 0 0 90>}]
 dyn_coat_fr_mid_03_l [{shapeType=Capsule, length=16, radius=3, shapeName=dyn_coat_fr_mid_03_l_capsule, center=<1.657 0 -0.0002>, rotation=<rot 0 0 90>}]
 dyn_coat_fr_mid_03_r [{shapeType=Capsule, length=12.25, radius=4.75, shapeName=dyn_coat_fr_mid_03_r_capsule, center=<-1.6448 0 -0.0001>, rotation=<rot 0 0 90>}]
 dyn_coat_fr_mid_04_l [{shapeType=Capsule, length=8.75, radius=5.25, shapeName=dyn_coat_fr_mid_04_l_capsule, center=<5.6628 0 -0.0002>, rotation=<rot 0 0 90>}]
 dyn_coat_fr_mid_04_r [{
   shapeType=Capsule
   length=9.25
   radius=4.75
   shapeName=dyn_coat_fr_mid_04_r_capsule
   center=<-4.1102 2.3015 0.0937>
   rotation=<rot 3.841 -10.4904 75.486>
  }]
 dyn_coat_out_01_l [{shapeType=Capsule, length=11, radius=3, center=<11.1341 0.0569 -1.9034>, rotation=<rot 0 0 90>}]
 dyn_coat_out_01_r [{shapeType=Capsule, length=11, radius=3, center=<-12.1716 -0.0001 0>, rotation=<rot 0 0 90>}]
 dyn_coat_out_02_l [{shapeType=Capsule, length=14.375, radius=3, center=<7.6136 0.0019 -0.0631>, rotation=<rot 0 0 90>}]
 dyn_coat_out_02_r [{shapeType=Capsule, length=21, radius=3, center=<-7.2603 0 -0.0001>, rotation=<rot 0 0 90>}]
 dyn_coat_out_03_l [{shapeType=Capsule, length=14.375, radius=3, center=<5.3134 0.0075 -0.2496>, rotation=<rot 0 0 90>}]
 dyn_coat_out_03_r [{shapeType=Capsule, length=21, radius=3, center=<-4.0642 0 -0.0001>, rotation=<rot 0 0 90>}]
 dyn_coat_out_04_l [{shapeType=Capsule, length=12.375, radius=5, center=<3.3292 0 -0.0001>, rotation=<rot 0 0 90>}]
 dyn_coat_out_04_r [{shapeType=Capsule, length=14, radius=3, center=<-5.1661 0 0>, rotation=<rot 0 0 90>}]
 dyn_pipe [{shapeType=Capsule, length=14.2998, radius=3.75, center=<8.5244 -0.6179 -0.1634>, rotation=<rot 88.4279 46.9563 56.5857>}]
 pelvis [{shapeType=Capsule, length=7.625, radius=12.625, center=<-1.2314 -1.823 0>}]
 spine_01 [
  {shapeType=Capsule, length=10.5, radius=12, center=<0.134 -0.7209 0>}
  {shapeType=Capsule, length=5.5, radius=3.5, center=<-1.1503 10.7537 -7.2445>, rotation=<rot 0.6319 0.1114 19.9906>}
 ]
 spine_02 [
  {shapeType=Capsule, length=9.375, radius=10.625, center=<0.0516 -0.7165 0>}
  {shapeType=Capsule, length=6.75, radius=9, center=<0 -4.8641 0>}
 ]
 spine_03 [{shapeType=Capsule, length=13.25, radius=10.25, center=<0 -0.7157 0>}]
 thigh_twist_01_l [
  {shapeType=Capsule, length=32.875, radius=10.75, center=<-0.1171 -1.1517 -1.0401>, rotation=<rot 86.963 179.7173 179.7253>}
  {shapeType=Capsule, length=24.75, radius=11.25, center=<1.1975 0.8044 1.1425>, rotation=<rot 84.9999 0 0>}
 ]
 thigh_twist_01_r [
  {shapeType=Capsule, length=32.875, radius=10.75, center=<0.1171 1.1517 1.0401>, rotation=<rot 86.963 179.7173 179.7253>}
  {shapeType=Capsule, length=24.75, radius=11.25, center=<-1.2587 -0.2975 -1.8421>, rotation=<rot 84.9999 0 0>}
 ]
}
constraints [
 {bone1Name=dyn_coat_bk_02, bone2Name=spine_03, swing1LimitDegrees=10, swing2LimitDegrees=20, twistLimitDegrees=1}
 {bone1Name=dyn_coat_bk_03_l, bone2Name=dyn_coat_bk_02_l, swing1LimitDegrees=5, swing2LimitDegrees=35, twistLimitDegrees=1}
 {bone1Name=dyn_coat_bk_04_l, bone2Name=dyn_coat_bk_03_l, swing1LimitDegrees=10, swing2LimitDegrees=45, twistLimitDegrees=1}
 {bone1Name=dyn_coat_bk_03_r, bone2Name=dyn_coat_bk_02_r, swing1LimitDegrees=5, swing2LimitDegrees=35, twistLimitDegrees=1}
 {bone1Name=dyn_coat_bk_04_r, bone2Name=dyn_coat_bk_03_r, swing1LimitDegrees=20, swing2LimitDegrees=45, twistLimitDegrees=1}
 {bone1Name=dyn_coat_bk_mid_02_l, bone2Name=dyn_coat_bk_mid_01_l, swing1LimitDegrees=10, swing2LimitDegrees=20, twistLimitDegrees=1}
 {bone1Name=dyn_coat_bk_mid_03_l, bone2Name=dyn_coat_bk_mid_02_l, swing1LimitDegrees=5, swing2LimitDegrees=35, twistLimitDegrees=1}
 {bone1Name=dyn_coat_bk_mid_04_l, bone2Name=dyn_coat_bk_mid_03_l, swing1LimitDegrees=5, swing2LimitDegrees=35, twistLimitDegrees=1}
 {bone1Name=dyn_coat_out_02_l, bone2Name=dyn_coat_out_01_l, swing1LimitDegrees=10, swing2LimitDegrees=20, twistLimitDegrees=1}
 {bone1Name=dyn_coat_out_03_l, bone2Name=dyn_coat_out_02_l, swing1LimitDegrees=5, swing2LimitDegrees=35, twistLimitDegrees=1}
 {bone1Name=dyn_coat_out_04_l, bone2Name=dyn_coat_out_03_l, swing1LimitDegrees=5, swing2LimitDegrees=45, twistLimitDegrees=1}
 {bone1Name=dyn_coat_fr_mid_02_l, bone2Name=dyn_coat_fr_mid_01_l, swing1LimitDegrees=10, swing2LimitDegrees=20, twistLimitDegrees=1}
 {bone1Name=dyn_coat_fr_mid_03_l, bone2Name=dyn_coat_fr_mid_02_l, swing1LimitDegrees=5, swing2LimitDegrees=35, twistLimitDegrees=1}
 {bone1Name=dyn_coat_fr_mid_04_l, bone2Name=dyn_coat_fr_mid_03_l, swing1LimitDegrees=45, swing2LimitDegrees=45, twistLimitDegrees=5}
 {bone1Name=dyn_coat_bk_mid_02_r, bone2Name=dyn_coat_bk_mid_01_r, swing1LimitDegrees=10, swing2LimitDegrees=20, twistLimitDegrees=1}
 {bone1Name=dyn_coat_bk_mid_03_r, bone2Name=dyn_coat_bk_mid_02_r, swing1LimitDegrees=5, swing2LimitDegrees=35, twistLimitDegrees=1}
 {bone1Name=dyn_coat_bk_mid_04_r, bone2Name=dyn_coat_bk_mid_03_r, swing1LimitDegrees=20, swing2LimitDegrees=45, twistLimitDegrees=1}
 {bone1Name=dyn_coat_out_02_r, bone2Name=dyn_coat_out_01_r, swing1LimitDegrees=10, swing2LimitDegrees=20, twistLimitDegrees=2}
 {bone1Name=dyn_coat_out_03_r, bone2Name=dyn_coat_out_02_r, swing1LimitDegrees=5, swing2LimitDegrees=35, twistLimitDegrees=2}
 {bone1Name=dyn_coat_out_04_r, bone2Name=dyn_coat_out_03_r, swing1LimitDegrees=5, swing2LimitDegrees=45, twistLimitDegrees=2}
 {bone1Name=dyn_coat_fr_mid_02_r, bone2Name=dyn_coat_fr_mid_01_r, swing1LimitDegrees=10, swing2LimitDegrees=20, twistLimitDegrees=1}
 {bone1Name=dyn_coat_fr_mid_03_r, bone2Name=dyn_coat_fr_mid_02_r, swing1LimitDegrees=5, swing2LimitDegrees=35, twistLimitDegrees=1}
 {bone1Name=dyn_coat_fr_mid_04_r, bone2Name=dyn_coat_fr_mid_03_r, swing1LimitDegrees=10, swing2LimitDegrees=45, twistLimitDegrees=1}
 {bone1Name=dyn_coat_bk_mid_01_l, bone2Name=spine_03, swing1LimitDegrees=2, swing2LimitDegrees=15, twistLimitDegrees=1}
 {bone1Name=dyn_coat_out_01_l, bone2Name=spine_03, swing1LimitDegrees=10, swing2LimitDegrees=15, twistLimitDegrees=1}
 {bone1Name=dyn_coat_fr_mid_01_l, bone2Name=spine_03, swing1LimitDegrees=10, swing2LimitDegrees=15, twistLimitDegrees=1}
 {bone1Name=dyn_coat_out_01_r, bone2Name=spine_03, swing1LimitDegrees=10, swing2LimitDegrees=15, twistLimitDegrees=2}
 {bone1Name=dyn_coat_fr_mid_01_r, bone2Name=spine_03, swing1LimitDegrees=10, swing2LimitDegrees=15, twistLimitDegrees=1}
 {
  bone1Name=thigh_twist_01_l
  bone2Name=pelvis
  swing1LimitDegrees=45
  swing1Motion=Free
  swing2LimitDegrees=45
  swing2Motion=Free
  twistLimitDegrees=45
  twistMotion=Free
 }
 {
  bone1Name=thigh_twist_01_r
  bone2Name=pelvis
  swing1LimitDegrees=45
  swing1Motion=Free
  swing2LimitDegrees=45
  swing2Motion=Free
  twistLimitDegrees=45
  twistMotion=Free
 }
 {bone1Name=dyn_pipe, bone2Name=pelvis, swing1LimitDegrees=45, swing2LimitDegrees=45, twistLimitDegrees=5}
 {
  bone1Name=spine_03
  bone2Name=spine_02
  swing1LimitDegrees=45
  swing1Motion=Free
  swing2LimitDegrees=45
  swing2Motion=Free
  twistLimitDegrees=45
  twistMotion=Free
 }
 {bone1Name=dyn_coat_fr_01_l, bone2Name=spine_03, swing1LimitDegrees=10, swing2LimitDegrees=90, twistLimitDegrees=90}
 {bone1Name=dyn_coat_fr_02_l, bone2Name=dyn_coat_fr_01_l, swing1LimitDegrees=10, swing2LimitDegrees=20, twistLimitDegrees=1}
 {bone1Name=dyn_coat_fr_03_l, bone2Name=dyn_coat_fr_02_l, swing1LimitDegrees=5, swing2LimitDegrees=35, twistLimitDegrees=1}
 {bone1Name=dyn_coat_fr_04_l, bone2Name=dyn_coat_fr_03_l, swing1LimitDegrees=5, swing2LimitDegrees=35, twistLimitDegrees=1}
 {bone1Name=dyn_coat_fr_05_l, bone2Name=dyn_coat_fr_04_l, swing1LimitDegrees=10, swing2LimitDegrees=35, twistLimitDegrees=1}
 {bone1Name=dyn_coat_fr_06_l, bone2Name=dyn_coat_fr_05_l, swing1LimitDegrees=10, swing2LimitDegrees=45, twistLimitDegrees=1}
 {bone1Name=dyn_coat_fr_01_r, bone2Name=spine_03, swing1LimitDegrees=10, swing2LimitDegrees=15, twistLimitDegrees=1}
 {bone1Name=dyn_coat_fr_02_r, bone2Name=dyn_coat_fr_01_r, swing1LimitDegrees=10, swing2LimitDegrees=15, twistLimitDegrees=2}
 {bone1Name=dyn_coat_fr_03_r, bone2Name=dyn_coat_fr_02_r, swing1LimitDegrees=5, swing2LimitDegrees=15, twistLimitDegrees=2}
 {bone1Name=dyn_coat_fr_04_r, bone2Name=dyn_coat_fr_03_r, swing1LimitDegrees=5, swing2LimitDegrees=20, twistLimitDegrees=2}
 {bone1Name=dyn_coat_fr_05_r, bone2Name=dyn_coat_fr_04_r, swing1LimitDegrees=10, swing2LimitDegrees=10, twistLimitDegrees=2}
 {bone1Name=dyn_coat_fr_06_r, bone2Name=dyn_coat_fr_05_r, swing1LimitDegrees=10, swing2LimitDegrees=10, twistLimitDegrees=2}
 {
  bone1Name=spine_01
  bone2Name=pelvis
  swing1LimitDegrees=45
  swing1Motion=Free
  swing2LimitDegrees=45
  swing2Motion=Free
  twistLimitDegrees=45
  twistMotion=Free
 }
 {
  bone1Name=spine_02
  bone2Name=spine_01
  swing1LimitDegrees=45
  swing1Motion=Free
  swing2LimitDegrees=45
  swing2Motion=Free
  twistLimitDegrees=45
  twistMotion=Free
 }
 {bone1Name=dyn_coat_bk_02_l, bone2Name=dyn_coat_bk_02, swing1LimitDegrees=10, swing2LimitDegrees=20, twistLimitDegrees=1}
 {bone1Name=dyn_coat_bk_02_r, bone2Name=dyn_coat_bk_02, swing1LimitDegrees=10, swing2LimitDegrees=20, twistLimitDegrees=1}
 {bone1Name=dyn_coat_bk_mid_01_r, bone2Name=spine_03, swing1LimitDegrees=10, swing2LimitDegrees=15, twistLimitDegrees=1}
]
