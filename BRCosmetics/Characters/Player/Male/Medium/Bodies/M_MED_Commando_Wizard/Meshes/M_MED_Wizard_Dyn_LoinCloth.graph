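# Physics asset for the dynamic loincloth bones on the male medium Commando Wizard body.
# Left and right three-bone capsule chains hang from their thigh twist bones, with tighter twist limits on the lower segments.
# Cross constraints tie each left segment to its right counterpart, linking the chains laterally.
# Larger capsules form the pelvis and thigh bodies, and the thighs join the pelvis through angularly free constraints.
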PhysicsAsset /BRCosmetics/Characters/Player/Male/Medium/Bodies/M_MED_Commando_Wizard/Meshes/M_MED_Wizard_Dyn_LoinCloth
bodies {
 dyn_loincloth_01_l [{shapeType=Capsule, length=8, radius=3, shapeName=dyn_loincloth_01_l_capsule, center=<5.8436 3.1005 0>, rotation=<rot 90 0 0>}]
 dyn_loincloth_01_r [{shapeType=Capsule, length=8, radius=3, shapeName=dyn_loincloth_01_r_capsule, center=<-5.8436 -2.3879 0>, rotation=<rot 90 0 0>}]
 dyn_loincloth_02_l [{shapeType=Capsule, length=8, radius=2, shapeName=dyn_loincloth_02_l_capsule, center=<5.8436 3.1005 0>, rotation=<rot 0 0 90>}]
 dyn_loincloth_02_r [{shapeType=Capsule, length=8, radius=2.5, shapeName=dyn_loincloth_02_r_capsule, center=<-5.8436 -2.3879 0>, rotation=<rot 0 0 90>}]
 dyn_loincloth_03_l [{shapeType=Capsule, length=8, radius=3, shapeName=dyn_loincloth_03_l_capsule, center=<5.8436 3.1005 0>, rotation=<rot 0 0 90>}]
 dyn_loincloth_03_r [{shapeType=Capsule, length=8, radius=3, shapeName=dyn_loincloth_03_r_capsule, center=<-5.8436 -2.3879 0>, rotation=<rot 0 0 90>}]
 pelvis [{shapeType=Capsule, length=10.875, radius=12.625, center=<0 -2.4517 0>}]
 thigh_twist_01_l [{shapeType=Capsule, length=24.75, radius=10.75, center=<1.0251 0.0148 -0.3465>, rotation=<rot 90 -90 -88.9984>}]
 thigh_twist_01_r [{shapeType=Capsule, length=24.75, radius=10.75, center=<-1.0251 -0.0148 0.3465>, rotation=<rot 90 -90 -88.9984>}]
}
constraints [
 {bone1Name=dyn_loincloth_01_l, bone2Name=thigh_twist_01_l, swing1LimitDegrees=45, swing2LimitDegrees=45, twistLimitDegrees=45}
 {bone1Name=dyn_loincloth_02_l, bone2Name=dyn_loincloth_01_l, swing1LimitDegrees=20, swing2LimitDegrees=45, twistLimitDegrees=5}
 {bone1Name=dyn_loincloth_03_l, bone2Name=dyn_loincloth_02_l, swing1LimitDegrees=20, swing2LimitDegrees=45, twistLimitDegrees=5}
 {bone1Name=dyn_loincloth_01_r, bone2Name=thigh_twist_01_r, swing1LimitDegrees=45, swing2LimitDegrees=45, twistLimitDegrees=45}
 {bone1Name=dyn_loincloth_02_r, bone2Name=dyn_loincloth_01_r, swing1LimitDegrees=20, swing2LimitDegrees=45, twistLimitDegrees=5}
 {bone1Name=dyn_loincloth_03_r, bone2Name=dyn_loincloth_02_r, swing1LimitDegrees=20, swing2LimitDegrees=45, twistLimitDegrees=5}
 {
  bone1Name=thigh_twist_01_l
  bone2Name=pelvis
  swing1LimitDegrees=45
  swing1Motion=Free
  swing2LimitDegrees=45
  swing2Motion=Free
  twistLimitDegrees=45
  twistMotion=Free
 }
 {
  bone1Name=thigh_twist_01_r
  bone2Name=pelvis
  swing1LimitDegrees=45
  swing1Motion=Free
  swing2LimitDegrees=45
  swing2Motion=Free
  twistLimitDegrees=45
  twistMotion=Free
 }
 {bone1Name=dyn_loincloth_01_r, bone2Name=dyn_loincloth_01_l, swing1LimitDegrees=45, swing2LimitDegrees=45, twistLimitDegrees=45}
 {bone1Name=dyn_loincloth_02_r, bone2Name=dyn_loincloth_02_l, swing1LimitDegrees=45, swing2LimitDegrees=45, twistLimitDegrees=45}
 {bone1Name=dyn_loincloth_03_r, bone2Name=dyn_loincloth_03_l, swing1LimitDegrees=45, swing2LimitDegrees=45, twistLimitDegrees=45}
]
