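# Physics asset for the medium male Commando Wizard torso, defining secondary-motion bodies for its scarf, collar, coat lapels and bottle.
# Capsules on the spine and left thigh give body volumes alongside the dynamic bones.
# Two three-link scarf chains hang from master bodies on spine_05, with limited swing and locked twist at each root.
# Collar and lapels have tight, asymmetric limits to upper spine bones, while bottle and thigh have no constraints.
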PhysicsAsset /BRCosmetics/Characters/Player/Male/Medium/Bodies/M_MED_Commando_Wizard/Meshes/M_MED_Wizard_Dyn_Torso
bodies {
 dyn_bottle [{shapeType=Capsule, length=11.816, radius=2.7338, center=<0.1022 -1.9036 0.064>, rotation=<rot 0 -93.0486 -81.0549>}]
 dyn_coat_lapel_l [{shapeType=Capsule, length=7, radius=2, center=<0.5 -0.0001 0>, rotation=<rot -90 0 90.0018>}]
 dyn_coat_lapel_r [{shapeType=Capsule, length=7, radius=2, center=<-0.5 -0.0001 0>, rotation=<rot -90 0 90.0018>}]
 dyn_collar_bk [{shapeType=Capsule, length=2.6562, radius=2.1875, center=<0.8985 -0.0001 0>, rotation=<rot -90 -179.9802 179.9803>}]
 dyn_collar_l [{shapeType=Capsule, length=2.6562, radius=2.1875, center=<0.8985 -0.0001 0>, rotation=<rot -90 -179.9802 179.9803>}]
 dyn_collar_r [{shapeType=Capsule, length=2.6562, radius=2.1875, center=<-1.2692 -0.0003 0>, rotation=<rot -90 -179.9802 179.9803>}]
 dyn_scarf_l_01 [{shapeType=Capsule, length=14, radius=2.8125, center=<7.2467 -0.0003 1.9>, rotation=<rot 90 0 0>}]
 dyn_scarf_l_02 [{shapeType=Capsule, length=12.3125, radius=2.8125, center=<6.5023 -0.0001 1.9>, rotation=<rot 90 0 0>}]
 dyn_scarf_l_03 [{shapeType=Capsule, length=10.5312, radius=2.8125, center=<5.3049 -0.0002 1.9>, rotation=<rot 90 0 0>}]
 dyn_scarf_l_MSTR [{shapeType=Capsule, length=0.505, radius=0.505, rotation=<rot 0 -90 0>}]
 dyn_scarf_r_01 [{shapeType=Capsule, length=14, radius=2.8125, center=<-7.2467 -0.0003 -1.9>, rotation=<rot 90 0 0>}]
 dyn_scarf_r_02 [{shapeType=Capsule, length=12.3125, radius=2.8125, center=<-6.5023 -0.0001 -1.1511>, rotation=<rot 90 0 0>}]
 dyn_scarf_r_03 [{shapeType=Capsule, length=4.375, radius=2.8125, center=<-2.6401 -0.0003 -1.9>, rotation=<rot 90 0 0>}]
 dyn_scarf_r_MSTR [{shapeType=Capsule, length=0.505, radius=0.505, rotation=<rot 0 -90 0>}]
 spine_02 [{shapeType=Capsule, length=5.75, radius=13.5}]
 spine_04 [{shapeType=Capsule, length=17.7188, radius=13.5, center=<-1.804 0.3796 0>}]
 spine_05 [
  {shapeType=Capsule, length=22.8438, radius=10.75, center=<-6.2433 -0.0219 0.0001>}
  {shapeType=Capsule, length=9.0938, radius=7.9062, center=<-4.5696 3.4977 0>}
 ]
 thigh_l [{shapeType=Capsule, length=32, radius=14.75, center=<-21.2585 -0.4532 -0.3465>, rotation=<rot 90 -90 -88.9984>}]
}
constraints [
 {bone1Name=dyn_scarf_l_02, bone2Name=dyn_scarf_l_01, swing1LimitDegrees=45, swing2LimitDegrees=45, twistLimitDegrees=5}
 {bone1Name=dyn_scarf_l_03, bone2Name=dyn_scarf_l_02, swing1LimitDegrees=45, swing2LimitDegrees=45, twistLimitDegrees=5}
 {bone1Name=dyn_scarf_r_02, bone2Name=dyn_scarf_r_01, swing1LimitDegrees=45, swing2LimitDegrees=45, twistLimitDegrees=5}
 {bone1Name=dyn_scarf_r_03, bone2Name=dyn_scarf_r_02, swing1LimitDegrees=45, swing2LimitDegrees=45, twistLimitDegrees=5}
 {bone1Name=dyn_scarf_l_01, bone2Name=dyn_scarf_l_MSTR, swing1LimitDegrees=45, swing2LimitDegrees=45, twistLimitDegrees=5, twistMotion=Locked}
 {bone1Name=dyn_scarf_r_01, bone2Name=dyn_scarf_r_MSTR, swing1LimitDegrees=45, swing2LimitDegrees=45, twistLimitDegrees=5, twistMotion=Locked}
 {bone1Name=dyn_collar_l, bone2Name=spine_05, swing1LimitDegrees=6, swing2LimitDegrees=25, twistLimitDegrees=0.1}
 {bone1Name=dyn_collar_r, bone2Name=spine_05, swing1LimitDegrees=6, swing2LimitDegrees=25, twistLimitDegrees=0.1}
 {bone1Name=dyn_collar_bk, bone2Name=spine_05, swing1LimitDegrees=6, swing2LimitDegrees=25, twistLimitDegrees=0.1}
 {
  bone1Name=spine_05
  bone2Name=spine_04
  swing1LimitDegrees=45
  swing1Motion=Free
  swing2LimitDegrees=45
  swing2Motion=Free
  twistLimitDegrees=45
  twistMotion=Free
 }
 {bone1Name=dyn_coat_lapel_l, bone2Name=spine_04, swing1LimitDegrees=0.1, swing2LimitDegrees=45, twistLimitDegrees=0.1}
 {bone1Name=dyn_coat_lapel_r, bone2Name=spine_04, swing1LimitDegrees=0.1, swing2LimitDegrees=45, twistLimitDegrees=0.1}
 {
  bone1Name=spine_04
  bone2Name=spine_02
  swing1LimitDegrees=45
  swing1Motion=Free
  swing2LimitDegrees=45
  swing2Motion=Free
  twistLimitDegrees=45
  twistMotion=Free
 }
 {
  bone1Name=dyn_scarf_l_MSTR
  bone2Name=spine_05
  swing1LimitDegrees=45
  swing1Motion=Free
  swing2LimitDegrees=45
  swing2Motion=Free
  twistLimitDegrees=45
  twistMotion=Free
 }
 {
  bone1Name=dyn_scarf_r_MSTR
  bone2Name=spine_05
  swing1LimitDegrees=45
  swing1Motion=Free
  swing2LimitDegrees=45
  swing2Motion=Free
  twistLimitDegrees=45
  twistMotion=Free
 }
]
