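# Post-process-style cosmetic AnimBP for the Wizard Commando body, adding secondary motion to coat, loincloth, scarf and accessories over poses copied from the parent mesh.
# RigidBody and AnimDynamics simulate these, with separate standing and crouch branches and corrective bone transforms.
# With `bRigidBodyEnabled` off, constraints pin coat bones to pelvis, spine and thighs instead.
# Pose Drivers and Flap nodes then shape sleeves, followed by aim and jump adjustments.
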AnimBlueprint /BRCosmetics/Characters/Player/Male/Medium/Bodies/M_MED_Commando_Wizard/Meshes/M_MED_Wizard_AnimBP
target_skeleton=&/M_MED_Wizard_Skeleton
graphs [
 AnimGraph {nodes [
   Anim Dynamics - Bone: dyn_pipe(Alpha=1, ExternalForce="0,0,0") {
    angularSpringConstant=1500
    bAngularSpring=true
    bGravityOverrideInSimSpace=true
    bUseGravityOverride=true
    bUseSphericalLimits=true
    alphaBoolBlend {blendOption=Linear}
    boundBone {boneName=dyn_pipe}
    componentLinearAccScale=<1 0.2 0.45>
    componentLinearVelScale=<1 2 0.45>
    gravityOverride {z=-980}
    physicsBodyDefinitions [{
      sphereCollisionRadius=0
      boundBone {boneName=dyn_pipe}
      boxExtents=<8 24 24>
      constraintSetup {angularLimitsMax=<2 20 5>, angularLimitsMin=<-3 -15 -20>, angularTarget=<0 0 0>}
      localJointOffset=<5 0 0>
     }]
    sphericalLimits [{limitRadius=10, drivingBone {boneName=thigh_r}, sphereLocalOffset=<0 0 0>}]
   }
   Component To Local()
   "Constraint - dyn_coat_fr_03_l (PPPPPP)"!(
    ComponentPose
    Alpha=1
    ConstraintWeights_0=0.0125
    ConstraintWeights_1:float=0.0762
    ConstraintWeights_2:float=0.0205
    ConstraintWeights_3:float=0.4792
    ConstraintWeights_4:float=0.3671
    ConstraintWeights_5:float=0.0445
   ) -> (Pose) {
    alphaBoolBlend {blendOption=Linear}
    boneToModify {boneName=dyn_coat_fr_03_l}
    constraintSetup [
     {targetBone {boneName=dyn_coat_fr_mid_02_l}}
     {targetBone {boneName=dyn_coat_fr_04_l}}
     {targetBone {boneName=dyn_coat_fr_mid_01_l}}
     {targetBone {boneName=dyn_coat_fr_03_l}}
     {targetBone {boneName=dyn_coat_fr_02_l}}
     {targetBone {boneName=dyn_coat_fr_01_l}}
    ]
    constraintWeights [1,1,1,1,1,1]
   }
   n1 = "Constraint - dyn_loincloth_01_r (PPP)"!(ComponentPose, Alpha=1, ConstraintWeights_0=0.6434, ConstraintWeights_1:float=0.3355, ConstraintWeights_2:float=0.0211) -> (Pose) {
    alphaBoolBlend {blendOption=Linear}
    boneToModify {boneName=dyn_loincloth_01_r}
    constraintSetup [{targetBone {boneName=pelvis}}, {targetBone {boneName=thigh_twist_01_r}}, {targetBone {boneName=dyn_loincloth_01_r}}]
    constraintWeights [1,1,1]
   }
   "Constraint - dyn_loincloth_02_r (PPP)"!(ComponentPose=@n1, Alpha=0.5, ConstraintWeights_0=0.0337, ConstraintWeights_1:float=0.2074, ConstraintWeights_2:float=0.7589) -> (Pose) {
    alphaBoolBlend {blendOption=Linear}
    boneToModify {boneName=dyn_loincloth_02_r}
    constraintSetup [{targetBone {boneName=thigh_r}}, {targetBone {boneName=pelvis}}, {targetBone {boneName=thigh_twist_01_r}}]
    constraintWeights [1,1,1]
   }
   n2 = "Constraint - dyn_loincloth_01_l (PPP)"!(ComponentPose, Alpha=1, ConstraintWeights_0=0.6434, ConstraintWeights_1:float=0.3355, ConstraintWeights_2:float=0.0211) -> (Pose) {
    alphaBoolBlend {blendOption=Linear}
    boneToModify {boneName=dyn_loincloth_01_l}
    constraintSetup [{targetBone {boneName=pelvis}}, {targetBone {boneName=thigh_twist_01_l}}, {targetBone {boneName=dyn_loincloth_01_l}}]
    constraintWeights [1,1,1]
   }
   n3 = "Constraint - dyn_loincloth_01_r (PPP)"(ComponentPose=@n2, Alpha=1, ConstraintWeights_0=0.6434, ConstraintWeights_1=0.3355, ConstraintWeights_2=0.0211) {
    alphaBoolBlend {blendOption=Linear}
    boneToModify {boneName=dyn_loincloth_01_r}
    constraintSetup [{targetBone {boneName=pelvis}}, {targetBone {boneName=thigh_twist_01_r}}, {targetBone {boneName=dyn_loincloth_01_r}}]
    constraintWeights [1,1,1]
   }
   n4 = "Constraint - dyn_loincloth_02_l (PPP)"!(ComponentPose=@n3, Alpha=1, ConstraintWeights_0=0.6434, ConstraintWeights_1:float=0.3355, ConstraintWeights_2:float=0.0211) -> (Pose) {
    alphaBoolBlend {blendOption=Linear}
    boneToModify {boneName=dyn_loincloth_02_l}
    constraintSetup [{targetBone {boneName=pelvis}}, {targetBone {boneName=thigh_twist_01_l}}, {targetBone {boneName=dyn_loincloth_01_l}}]
    constraintWeights [1,1,1]
   }
   "Constraint - dyn_loincloth_02_r (PPP)"(ComponentPose=@n4, Alpha=1, ConstraintWeights_0=0.6434, ConstraintWeights_1=0.3355, ConstraintWeights_2=0.0211) {
    alphaBoolBlend {blendOption=Linear}
    boneToModify {boneName=dyn_loincloth_02_r}
    constraintSetup [{targetBone {boneName=pelvis}}, {targetBone {boneName=thigh_twist_01_r}}, {targetBone {boneName=dyn_loincloth_01_r}}]
    constraintWeights [1,1,1]
   }
   Copy Pose From Mesh() {bUseAttachedParent=true}
   Copy Pose From Mesh() {bUseAttachedParent=true}
   n5 = "Transform (Modify) Bone - Bone: dyn_coat_bk_mid_01_l"(Alpha=1, Translation="-1.937,-0.079,1.366", Rotation="10,0.000,0.000", Scale="1,1,1") {
    rotationMode=BMM_Additive
    rotationSpace=BCS_BoneSpace
    scaleMode=BMM_Additive
    translationMode=BMM_Additive
    translationSpace=BCS_BoneSpace
    alphaBoolBlend {blendOption=Linear}
    alphaScaleBiasClamp {interpSpeedDecreasing=20, interpSpeedIncreasing=20}
    boneToModify {boneName=dyn_coat_bk_mid_01_l}
    rotation {pitch=10}
    translation=<-1.937 -0.079 1.366>
   }
   "Transform (Modify) Bone - Bone: dyn_coat_bk_mid_01_r"(ComponentPose=@n5, Alpha=1, Translation="-0.304,-0.000,-2.815", Rotation="10,0.000,0.000", Scale="1,1,1") {
    rotationMode=BMM_Additive
    rotationSpace=BCS_BoneSpace
    scaleMode=BMM_Additive
    translationMode=BMM_Additive
    translationSpace=BCS_BoneSpace
    alphaBoolBlend {blendOption=Linear}
    alphaScaleBiasClamp {interpSpeedDecreasing=20, interpSpeedIncreasing=20}
    boneToModify {boneName=dyn_coat_bk_mid_01_r}
    rotation {pitch=10}
    translation {x=-0.304, z=-2.815}
   }
   n6 = "Constraint - dyn_scarf_r_MSTR (PPP)"!(
    ComponentPose=Local To Component(LocalPose=Use cached pose 'AnimCache'())
    Alpha=1
    ConstraintWeights_0=0.1185
    ConstraintWeights_1:float=0.1412
    ConstraintWeights_2:float=0.7403
   ) -> (Pose) {
    alphaBoolBlend {blendOption=Linear}
    boneToModify {boneName=dyn_scarf_r_MSTR}
    constraintSetup [{targetBone {boneName=clavicle_r}}, {targetBone {boneName=spine_04}}, {targetBone {boneName=spine_05}}]
    constraintWeights [1,1,1]
   }
   n7 = "Constraint - dyn_scarf_l_MSTR (PPP)"!(ComponentPose=@n6, Alpha=1, ConstraintWeights_0=0.1185, ConstraintWeights_1:float=0.1412, ConstraintWeights_2:float=0.7403) -> (Pose) {
    alphaBoolBlend {blendOption=Linear}
    boneToModify {boneName=dyn_scarf_l_MSTR}
    constraintSetup [{targetBone {boneName=clavicle_l}}, {targetBone {boneName=spine_04}}, {targetBone {boneName=spine_05}}]
    constraintWeights [1,1,1]
   }
   n8 = Get GravityOverride()
   n9 = RigidBody(ComponentPose=@n7, Alpha=Get FrontEnd Alpha() -> (FrontEnd Alpha:double), ExternalForce=@n8) {
    bOverrideWorldGravity=true
    overridePhysicsAsset=&/M_MED_Wizard_Dyn_Coat
    simulationSpace=BaseBoneSpace
    alphaBoolBlend {blendOption=Linear}
    baseBoneRef {boneName=spine_02}
    componentAppliedLinearAccClamp=<500 500 500>
    componentLinearAccScale=<1 0.7 0.7>
    componentLinearVelScale=<1 0.7 0.7>
    overrideWorldGravity {z=-925}
   }
   n10 = RigidBody(ComponentPose=@n9, Alpha=Get Forward_direct() -> (Forward_direct:double), ExternalForce=@n8) {
    bOverrideWorldGravity=true
    overridePhysicsAsset=&/M_MED_Wizard_Dyn_LoinCloth
    simulationSpace=BaseBoneSpace
    alphaBoolBlend {blendOption=Linear}
    alphaScaleBiasClamp {interpSpeedDecreasing=5}
    baseBoneRef {boneName=pelvis}
    componentAppliedLinearAccClamp=<500 500 500>
    componentLinearAccScale=<1 0.7 0.7>
    componentLinearVelScale=<1 0.7 0.7>
    overrideWorldGravity {z=-980}
   }
   n11 = Get Off_Crouch() -> (Off_Crouch:double)
   n12 = "Transform (Modify) Bone - Bone: dyn_loincloth_01_l"!(ComponentPose=@n10, Alpha=@n11, Rotation="0.000,10,0.000") -> (Pose) {
    rotationMode=BMM_Additive
    rotationSpace=BCS_BoneSpace
    alphaBoolBlend {blendOption=Linear}
    boneToModify {boneName=dyn_loincloth_01_l}
    rotation {yaw=10}
   }
   n13 = "Transform (Modify) Bone - Bone: dyn_loincloth_01_r"!(ComponentPose=@n12, Alpha=@n11, Rotation="0,10,0") -> (Pose) {
    rotationMode=BMM_Additive
    rotationSpace=BCS_BoneSpace
    alphaBoolBlend {blendOption=Linear}
    boneToModify {boneName=dyn_loincloth_01_r}
    rotation {yaw=10}
   }
   n14 = Component To Local(ComponentPose=@n13)
   n15 = Get is_crouching() -> (is_crouching:bool)
   n16 = RigidBody(ComponentPose=Local To Component(LocalPose=Use cached pose 'Crouch Coat'())) {
    bOverrideWorldGravity=true
    overridePhysicsAsset=&/M_MED_Wizard_Dyn_CrouchLoinCloth
    simulationSpace=BaseBoneSpace
    alphaBoolBlend {blendOption=Linear}
    baseBoneRef {boneName=pelvis}
    componentAppliedLinearAccClamp=<200 200 200>
    componentLinearAccScale=<1 0.7 0.7>
    componentLinearVelScale=<1 0.7 0.7>
    overrideWorldGravity {z=-980}
   }
   n17 = "Transform (Modify) Bone - Bone: dyn_loincloth_01_r"!(ComponentPose=@n16, Alpha=1, Translation:Vector="0,0,0", Rotation="0,10,0", Scale:Vector="1,1,1") -> (Pose) {
    rotationMode=BMM_Additive
    rotationSpace=BCS_BoneSpace
    scaleMode=BMM_Additive
    translationMode=BMM_Additive
    translationSpace=BCS_BoneSpace
    alphaBoolBlend {blendOption=Linear}
    alphaScaleBiasClamp {interpSpeedDecreasing=20, interpSpeedIncreasing=20}
    boneToModify {boneName=dyn_loincloth_01_r}
    rotation {yaw=10}
   }
   n18 = "Transform (Modify) Bone - Bone: dyn_loincloth_01_l"!(ComponentPose=@n17, Alpha=1, Translation:Vector="-7,-7.265,15.289", Rotation="2.79,39.23,73.43", Scale:Vector="1,1,1") -> (Pose) {
    alpha=0
    rotationMode=BMM_Additive
    rotationSpace=BCS_BoneSpace
    translationMode=BMM_Additive
    translationSpace=BCS_BoneSpace
    alphaBoolBlend {blendOption=Linear}
    boneToModify {boneName=dyn_loincloth_01_l}
    rotation=<rot 2.79 39.23 73.43>
    translation=<-7 -7.265 15.289>
   }
   n19 = RigidBody(ComponentPose=Local To Component(LocalPose=Use cached pose 'Crouch Coat'()), Alpha=0.5) {
    bOverrideWorldGravity=true
    overridePhysicsAsset=&M_MED_Wizard_Dyn_CrouchLoinCloth
    simulationSpace=BaseBoneSpace
    alphaBoolBlend {blendOption=Linear}
    baseBoneRef {boneName=pelvis}
    componentAppliedLinearAccClamp=<500 500 500>
    componentLinearAccScale=<1 0.7 0.7>
    componentLinearVelScale=<1 0.7 0.7>
    overrideWorldGravity {z=-980}
   }
   n20 = "Transform (Modify) Bone - Bone: dyn_pipe"(ComponentPose=@n19, Alpha=Get Right_direct() -> (Right_direct:double), Translation="0,0,0", Rotation="-15,-15,0", Scale="1,1,1") {
    rotationMode=BMM_Additive
    rotationSpace=BCS_BoneSpace
    scaleMode=BMM_Additive
    translationMode=BMM_Additive
    translationSpace=BCS_BoneSpace
    alphaBoolBlend {blendOption=Linear}
    alphaScaleBiasClamp {interpSpeedDecreasing=20, interpSpeedIncreasing=20}
    boneToModify {boneName=dyn_pipe}
    rotation {pitch=-15, yaw=-15}
   }
   n21 = "Transform (Modify) Bone - Bone: dyn_pipe"(ComponentPose=@n20, Alpha=Get Left_direct() -> (Left_direct:double), Translation="0,0,0", Rotation="12,-10,0", Scale="1,1,1") {
    rotationMode=BMM_Additive
    rotationSpace=BCS_BoneSpace
    scaleMode=BMM_Additive
    translationMode=BMM_Additive
    translationSpace=BCS_BoneSpace
    alphaBoolBlend {blendOption=Linear}
    alphaScaleBiasClamp {interpSpeedDecreasing=20, interpSpeedIncreasing=20}
    boneToModify {boneName=dyn_pipe}
    rotation {pitch=12, yaw=-10}
   }
   n22 = "Constraint - dyn_pipe (P)"(ComponentPose=@n21, Alpha=0.4, ConstraintWeights_0=1) {
    alphaBoolBlend {blendOption=Linear}
    boneToModify {boneName=dyn_pipe}
    constraintSetup [{targetBone {boneName=thigh_r}}]
    constraintWeights [1]
   }
   n23 = Two Way Blend(A=Component To Local(ComponentPose=@n22), B=Component To Local(ComponentPose=@n18), Alpha=Get CrouchWalking() -> (CrouchWalking:double)) {blendNode {alphaBoolBlend {blendOption=Linear}, alphaScaleBiasClamp {bInterpResult=true, interpSpeedDecreasing=20}}}
   n24 = RigidBody(
    ComponentPose=Local To Component(LocalPose=Blend Poses by bool(BlendPose_0=@n23, BlendPose_1=@n14, BlendTime_0=0.2, BlendTime_1=1.4, bActiveValue=@n15) {blendType=Linear})
    ExternalForce=@n8
   ) {
    bOverrideWorldGravity=true
    overridePhysicsAsset=&/M_MED_Wizard_Dyn_Torso
    simulationSpace=BaseBoneSpace
    alphaBoolBlend {blendOption=Linear}
    baseBoneRef {boneName=spine_05}
    componentAppliedLinearAccClamp=<100 100 100>
    componentLinearAccScale=<0.1 0.1 0.1>
    componentLinearVelScale=<10 10 10>
    overrideWorldGravity {z=-980}
   }
   n25 = Anim Dynamics - Bone: dyn_bottle(ComponentPose=@n24, Alpha=0.65, ExternalForce="0,0,0") {
    angularSpringConstant=100
    bAngularSpring=true
    bGravityOverrideInSimSpace=true
    bUseGravityOverride=true
    alphaBoolBlend {blendOption=Linear}
    boundBone {boneName=dyn_bottle}
    componentAppliedLinearAccClamp=<800 800 800>
    componentLinearAccScale=<0.3 0.2 0.3>
    componentLinearVelScale=<0.2 0.2 0.3>
    gravityOverride {z=-980}
    physicsBodyDefinitions [{
      sphereCollisionRadius=0
      boundBone {boneName=dyn_bottle}
      boxExtents=<16 24 24>
      constraintSetup {angularLimitsMax=<0 15 2>, angularLimitsMin=<0 -15 0>, angularTarget=<0 0 0>}
      localJointOffset=<6.5 0 0>
     }]
   }
   n26 = Anim Dynamics - Bone: dyn_pipe(ComponentPose=@n25, Alpha=0.8, ExternalForce="0,0,0") {
    angularSpringConstant=100
    bAngularSpring=true
    bGravityOverrideInSimSpace=true
    bUseGravityOverride=true
    bUseSphericalLimits=true
    alphaBoolBlend {blendOption=Linear}
    boundBone {boneName=dyn_pipe}
    componentAppliedLinearAccClamp=<800 800 800>
    componentLinearAccScale=<0.3 0.2 1>
    componentLinearVelScale=<0.3 0.2 1>
    gravityOverride {z=-980}
    physicsBodyDefinitions [{
      sphereCollisionRadius=0
      boundBone {boneName=dyn_pipe}
      boxExtents=<16 24 24>
      constraintSetup {angularLimitsMax=<0 2 2>, angularLimitsMin=<0 -5 -15>, angularTarget=<0 0 0>}
      localJointOffset=<6.5 0 0>
     }]
    sphericalLimits [
     {limitRadius=3.5, drivingBone {boneName=dyn_loincloth_01_r}, sphereLocalOffset=<0 0 0>}
     {limitRadius=10, drivingBone {boneName=thigh_r}, sphereLocalOffset=<0 -1 0>}
    ]
   }
   n27 = Anim Dynamics - Bone: dyn_necklace_01(ComponentPose=@n26, Alpha=1, ExternalForce="0,0,0") {
    angularDampingOverride=1
    angularSpringConstant=1500
    bAngularSpring=true
    bGravityOverrideInSimSpace=true
    bOverrideAngularDamping=true
    bUseGravityOverride=true
    gravityScale=0.35
    simulationSpace=RootRelative
    alphaBoolBlend {blendOption=Linear}
    alphaScaleBiasClamp {scale=0.6}
    boundBone {boneName=dyn_necklace_01}
    componentLinearAccScale=<0.3 0.3 0.45>
    componentLinearVelScale=<0.3 0.3 0.45>
    gravityOverride {z=-980}
    physicsBodyDefinitions [{
      sphereCollisionRadius=0
      boundBone {boneName=dyn_necklace_01}
      boxExtents=<10 20 20>
      constraintSetup {angularLimitsMax=<2 2 30>, angularLimitsMin=<-2 -30 -40>, angularTarget=<0 0 0>}
      localJointOffset=<10 0 0>
     }]
   }
   n28 = Anim Dynamics - Bone: dyn_cowl(ComponentPose=@n27, Alpha=1, ExternalForce="0,0,0") {
    angularDampingOverride=1
    angularSpringConstant=1500
    bAngularSpring=true
    bGravityOverrideInSimSpace=true
    bOverrideAngularDamping=true
    bUseGravityOverride=true
    gravityScale=0.35
    simulationSpace=RootRelative
    alphaBoolBlend {blendOption=Linear}
    alphaScaleBiasClamp {scale=0.4}
    boundBone {boneName=dyn_cowl}
    componentAppliedLinearAccClamp {z=10000}
    componentLinearAccScale {z=0.46}
    componentLinearVelScale {z=0.46}
    gravityOverride {z=-980}
    physicsBodyDefinitions [{
      sphereCollisionRadius=0
      boundBone {boneName=dyn_cowl}
      boxExtents=<10 20 20>
      constraintSetup {angularLimitsMax=<0 2 30>, angularLimitsMin=<0 -20 -40>, angularTarget=<0 0 0>}
      localJointOffset=<10 0 0>
     }]
   }
   n29 = "Transform (Modify) Bone - Bone: dyn_coat_bk_02_l"(ComponentPose=@n28, Alpha=1, Translation="0,0,0", Rotation="0,10,0", Scale="1,1,1") {
    rotationMode=BMM_Additive
    rotationSpace=BCS_BoneSpace
    scaleMode=BMM_Additive
    translationMode=BMM_Additive
    translationSpace=BCS_BoneSpace
    alphaBoolBlend {blendOption=Linear}
    alphaScaleBiasClamp {interpSpeedDecreasing=20, interpSpeedIncreasing=20}
    boneToModify {boneName=dyn_coat_bk_02_l}
    rotation {yaw=10}
   }
   n30 = "Transform (Modify) Bone - Bone: dyn_coat_bk_02_r"(ComponentPose=@n29, Alpha=1, Translation="0,0,0", Rotation="0.000,10,0.000", Scale="1,1,1") {
    rotationMode=BMM_Additive
    rotationSpace=BCS_BoneSpace
    scaleMode=BMM_Additive
    translationMode=BMM_Additive
    translationSpace=BCS_BoneSpace
    alphaBoolBlend {blendOption=Linear}
    alphaScaleBiasClamp {interpSpeedDecreasing=20, interpSpeedIncreasing=20}
    boneToModify {boneName=dyn_coat_bk_02_r}
    rotation {yaw=10}
   }
   n31 = "Transform (Modify) Bone - Bone: dyn_coat_out_01_r"(ComponentPose=@n30, Alpha=1, Translation="0.5,-0.000,-1", Rotation="10,0.000,0.000", Scale="1,1,1") {
    alpha=0
    rotationMode=BMM_Additive
    rotationSpace=BCS_BoneSpace
    scaleMode=BMM_Additive
    translationMode=BMM_Additive
    translationSpace=BCS_BoneSpace
    alphaBoolBlend {blendOption=Linear}
    alphaScaleBiasClamp {interpSpeedDecreasing=20, interpSpeedIncreasing=20}
    boneToModify {boneName=dyn_coat_out_01_r}
    rotation {pitch=10}
    translation {x=0.5, z=-1}
   }
   n32 = "Transform (Modify) Bone - Bone: dyn_coat_out_01_l"(ComponentPose=@n31, Alpha=1, Translation="-0.228,0.000,1.292", Rotation="10,0.000,0.000", Scale="1,1,1") {
    alpha=0
    rotationMode=BMM_Additive
    rotationSpace=BCS_BoneSpace
    scaleMode=BMM_Additive
    translationMode=BMM_Additive
    translationSpace=BCS_BoneSpace
    alphaBoolBlend {blendOption=Linear}
    alphaScaleBiasClamp {interpSpeedDecreasing=20, interpSpeedIncreasing=20}
    boneToModify {boneName=dyn_coat_out_01_l}
    rotation {pitch=10}
    translation=<-0.228 0.0003 1.2922>
   }
   n33 = "Transform (Modify) Bone - Bone: dyn_coat_bk_mid_03_r"(ComponentPose=@n32, Alpha=1, Translation="0,0,0", Rotation="5,0.000,0.000", Scale="1,1,1") {
    alpha=0
    rotationMode=BMM_Additive
    rotationSpace=BCS_BoneSpace
    scaleMode=BMM_Additive
    translationMode=BMM_Additive
    translationSpace=BCS_BoneSpace
    alphaBoolBlend {blendOption=Linear}
    alphaScaleBiasClamp {interpSpeedDecreasing=20, interpSpeedIncreasing=20}
    boneToModify {boneName=dyn_coat_bk_mid_03_r}
    rotation {pitch=5}
   }
   n34 = "Transform (Modify) Bone - Bone: dyn_coat_bk_mid_03_l"(ComponentPose=@n33, Alpha=1, Translation="0,0,0", Rotation="5,0.000,0.000", Scale="1,1,1") {
    alpha=0
    rotationMode=BMM_Additive
    rotationSpace=BCS_BoneSpace
    scaleMode=BMM_Additive
    translationMode=BMM_Additive
    translationSpace=BCS_BoneSpace
    alphaBoolBlend {blendOption=Linear}
    alphaScaleBiasClamp {interpSpeedDecreasing=20, interpSpeedIncreasing=20}
    boneToModify {boneName=dyn_coat_bk_mid_03_l}
    rotation {pitch=5}
   }
   n35 = "Transform (Modify) Bone - Bone: dyn_coat_bk_mid_01_l"(ComponentPose=@n34, Alpha=1, Translation="-1.937,-0.079,1.366", Rotation="10,0.000,0.000", Scale="1,1,1") {
    alpha=0
    rotationMode=BMM_Additive
    rotationSpace=BCS_BoneSpace
    scaleMode=BMM_Additive
    translationMode=BMM_Additive
    translationSpace=BCS_BoneSpace
    alphaBoolBlend {blendOption=Linear}
    alphaScaleBiasClamp {interpSpeedDecreasing=20, interpSpeedIncreasing=20}
    boneToModify {boneName=dyn_coat_bk_mid_01_l}
    rotation {pitch=10}
    translation=<-1.937 -0.079 1.366>
   }
   n36 = "Transform (Modify) Bone - Bone: dyn_coat_bk_mid_01_r"(ComponentPose=@n35, Alpha=1, Translation="-0.304,-0.000,-2.815", Rotation="10,0.000,0.000", Scale="1,1,1") {
    alpha=0
    rotationMode=BMM_Additive
    rotationSpace=BCS_BoneSpace
    scaleMode=BMM_Additive
    translationMode=BMM_Additive
    translationSpace=BCS_BoneSpace
    alphaBoolBlend {blendOption=Linear}
    alphaScaleBiasClamp {interpSpeedDecreasing=20, interpSpeedIncreasing=20}
    boneToModify {boneName=dyn_coat_bk_mid_01_r}
    rotation {pitch=10}
    translation {x=-0.304, z=-2.815}
   }
   n37 = "Transform (Modify) Bone - Bone: dyn_coat_fr_02_l"(ComponentPose=@n36, Alpha=1, Translation="0,0,1.5", Rotation="9.847,-10.152,-1.753", Scale="1,1,1") {
    alpha=0
    rotationMode=BMM_Replace
    rotationSpace=BCS_BoneSpace
    scaleMode=BMM_Additive
    translationMode=BMM_Additive
    translationSpace=BCS_BoneSpace
    alphaBoolBlend {blendOption=Linear}
    alphaScaleBiasClamp {bClampResult=true, bInterpResult=true, clampMax=0.8, interpSpeedDecreasing=2, interpSpeedIncreasing=2}
    boneToModify {boneName=dyn_coat_fr_02_l}
    rotation=<rot 9.847 -10.1521 -1.753>
    translation {z=1.5}
   }
   n38 = "Transform (Modify) Bone - Bone: dyn_coat_fr_04_l"(ComponentPose=@n37, Alpha=1, Translation="-1.152,-3.153,-0.091", Rotation="6.368,-26.003,-35", Scale="1,1,1") {
    alpha=0
    rotationMode=BMM_Replace
    rotationSpace=BCS_BoneSpace
    scaleMode=BMM_Additive
    translationMode=BMM_Additive
    translationSpace=BCS_BoneSpace
    alphaBoolBlend {blendOption=Linear}
    alphaScaleBiasClamp {bClampResult=true, bInterpResult=true, clampMax=0.8, interpSpeedDecreasing=2, interpSpeedIncreasing=2}
    boneToModify {boneName=dyn_coat_fr_04_l}
    rotation=<rot 6.368 -26.003 -35>
    translation=<-1.152 -3.153 -0.091>
   }
   n39 = "Transform (Modify) Bone - Bone: dyn_coat_fr_02_r"(ComponentPose=@n38, Alpha=1, Translation="2.908,3.031,-3.711", Rotation="19.684,-10.628,16.383", Scale="1,1,1") {
    alpha=0
    rotationMode=BMM_Replace
    rotationSpace=BCS_BoneSpace
    scaleMode=BMM_Additive
    translationMode=BMM_Additive
    translationSpace=BCS_BoneSpace
    alphaBoolBlend {blendOption=Linear}
    alphaScaleBiasClamp {bClampResult=true, bInterpResult=true, clampMax=0.8, interpSpeedDecreasing=2, interpSpeedIncreasing=2}
    boneToModify {boneName=dyn_coat_fr_02_r}
    rotation=<rot 19.684 -10.6281 16.3831>
    translation=<2.908 3.031 -3.711>
   }
   n40 = "Transform (Modify) Bone - Bone: dyn_coat_fr_03_r"(ComponentPose=@n39, Alpha=1, Translation="1.414,4.481,0.580", Rotation="-4.452,-6.632,-33.518", Scale="1,1,1") {
    alpha=0
    rotationMode=BMM_Replace
    rotationSpace=BCS_BoneSpace
    scaleMode=BMM_Additive
    translationMode=BMM_Additive
    translationSpace=BCS_BoneSpace
    alphaBoolBlend {blendOption=Linear}
    alphaScaleBiasClamp {bClampResult=true, bInterpResult=true, clampMax=0.8, interpSpeedDecreasing=2, interpSpeedIncreasing=2}
    boneToModify {boneName=dyn_coat_fr_03_r}
    rotation=<rot -4.452 -6.6321 -33.5182>
    translation=<1.414 4.481 0.58>
   }
   n41 = "Transform (Modify) Bone - Bone: dyn_pipe"(ComponentPose=@n40, Alpha=1, Translation="0,0,0", Rotation="-11.285,-6.289,-26.831", Scale="1,1,1") {
    alpha=0
    rotationMode=BMM_Replace
    rotationSpace=BCS_BoneSpace
    scaleMode=BMM_Additive
    translationMode=BMM_Additive
    translationSpace=BCS_BoneSpace
    alphaBoolBlend {blendOption=Linear}
    alphaScaleBiasClamp {bClampResult=true, bInterpResult=true, clampMax=0.8, interpSpeedDecreasing=2, interpSpeedIncreasing=2}
    boneToModify {boneName=dyn_pipe}
    rotation=<rot -11.285 -6.289 -26.8308>
   }
   n42 = "Constraint - dyn_pipe (P)"(ComponentPose=@n41, Alpha=Get Forward_motion() -> (Forward_motion:double), ConstraintWeights_0=1) {
    alphaBoolBlend {blendOption=Linear}
    alphaScaleBiasClamp {bInterpResult=true}
    boneToModify {boneName=dyn_pipe}
    constraintSetup [{targetBone {boneName=thigh_r}}]
    constraintWeights [1]
   }
   n43 = "Transform (Modify) Bone - Bone: dyn_pipe"(ComponentPose=@n42, Alpha=1, Translation="0,0,-0.1", Rotation="-10,-5,0", Scale="1,1,1") {
    alpha=0
    rotationMode=BMM_Replace
    rotationSpace=BCS_BoneSpace
    scaleMode=BMM_Additive
    translationMode=BMM_Additive
    translationSpace=BCS_BoneSpace
    alphaBoolBlend {blendOption=Linear}
    alphaScaleBiasClamp {bClampResult=true, bInterpResult=true, clampMax=0.8, interpSpeedDecreasing=2, interpSpeedIncreasing=2}
    boneToModify {boneName=dyn_pipe}
    rotation {pitch=-10, yaw=-5}
    translation {z=-0.1}
   }
   n44 = "Constraint - dyn_coat_fr_03_l (P)"(ComponentPose=Local To Component(LocalPose=Use cached pose 'AnimCache'()), Alpha=1, ConstraintWeights_0=1) {
    alphaBoolBlend {blendOption=Linear}
    boneToModify {boneName=dyn_coat_fr_03_l}
    constraintSetup [{targetBone {boneName=pelvis}}]
    constraintWeights [1]
   }
   n45 = "Constraint - dyn_coat_fr_mid_02_l (P)"(ComponentPose=@n44, Alpha=1, ConstraintWeights_0=1) {
    alphaBoolBlend {blendOption=Linear}
    boneToModify {boneName=dyn_coat_fr_mid_02_l}
    constraintSetup [{targetBone {boneName=pelvis}}]
    constraintWeights [1]
   }
   n46 = "Constraint - dyn_coat_out_02_l (P)"(ComponentPose=@n45, Alpha=1, ConstraintWeights_0=1) {
    alphaBoolBlend {blendOption=Linear}
    boneToModify {boneName=dyn_coat_out_02_l}
    constraintSetup [{targetBone {boneName=pelvis}}]
    constraintWeights [1]
   }
   n47 = "Constraint - dyn_coat_bk_mid_02_l (P)"(ComponentPose=@n46, Alpha=1, ConstraintWeights_0=1) {
    alphaBoolBlend {blendOption=Linear}
    boneToModify {boneName=dyn_coat_bk_mid_02_l}
    constraintSetup [{targetBone {boneName=pelvis}}]
    constraintWeights [1]
   }
   n48 = "Constraint - dyn_coat_bk_02 (P)"(ComponentPose=@n47, Alpha=1, ConstraintWeights_0=1) {
    alphaBoolBlend {blendOption=Linear}
    boneToModify {boneName=dyn_coat_bk_02}
    constraintSetup [{targetBone {boneName=pelvis}}]
    constraintWeights [1]
   }
   n49 = "Constraint - dyn_coat_bk_mid_02_r (P)"(ComponentPose=@n48, Alpha=1, ConstraintWeights_0=1) {
    alphaBoolBlend {blendOption=Linear}
    boneToModify {boneName=dyn_coat_bk_mid_02_r}
    constraintSetup [{targetBone {boneName=pelvis}}]
    constraintWeights [1]
   }
   n50 = "Constraint - dyn_coat_out_02_r (P)"(ComponentPose=@n49, Alpha=1, ConstraintWeights_0=1) {
    alphaBoolBlend {blendOption=Linear}
    boneToModify {boneName=dyn_coat_out_02_r}
    constraintSetup [{targetBone {boneName=pelvis}}]
    constraintWeights [1]
   }
   n51 = "Constraint - dyn_coat_fr_03_r (PP)"!(ComponentPose=@n50, Alpha=1, ConstraintWeights_0=0.3, ConstraintWeights_1:float=0.7) -> (Pose) {
    alphaBoolBlend {blendOption=Linear}
    boneToModify {boneName=dyn_coat_fr_03_r}
    constraintSetup [{targetBone {boneName=spine_01}}, {targetBone {boneName=pelvis}}]
    constraintWeights [1,1]
   }
   n52 = "Constraint - dyn_coat_fr_01_r (P)"(ComponentPose=@n51, Alpha=1, ConstraintWeights_0=1) {
    alphaBoolBlend {blendOption=Linear}
    boneToModify {boneName=dyn_coat_fr_01_r}
    constraintSetup [{targetBone {boneName=pelvis}}]
    constraintWeights [1]
   }
   n53 = "Constraint - dyn_coat_fr_04_l (P)"(ComponentPose=@n52, Alpha=1, ConstraintWeights_0=1) {
    alphaBoolBlend {blendOption=Linear}
    boneToModify {boneName=dyn_coat_fr_04_l}
    constraintSetup [{targetBone {boneName=thigh_l}}]
    constraintWeights [1]
   }
   n54 = "Constraint - dyn_coat_fr_05_l (P)"(ComponentPose=@n53, Alpha=1, ConstraintWeights_0=1) {
    alphaBoolBlend {blendOption=Linear}
    boneToModify {boneName=dyn_coat_fr_05_l}
    constraintSetup [{targetBone {boneName=thigh_l}}]
    constraintWeights [1]
   }
   n55 = "Constraint - dyn_coat_fr_mid_03_l (P)"(ComponentPose=@n54, Alpha=1, ConstraintWeights_0=1) {
    alphaBoolBlend {blendOption=Linear}
    boneToModify {boneName=dyn_coat_fr_mid_03_l}
    constraintSetup [{targetBone {boneName=thigh_l}}]
    constraintWeights [1]
   }
   n56 = "Constraint - dyn_coat_out_03_l (P)"(ComponentPose=@n55, Alpha=1, ConstraintWeights_0=1) {
    alphaBoolBlend {blendOption=Linear}
    boneToModify {boneName=dyn_coat_out_03_l}
    constraintSetup [{targetBone {boneName=thigh_l}}]
    constraintWeights [1]
   }
   n57 = "Constraint - dyn_coat_fr_04_r (P)"(ComponentPose=@n56, Alpha=1, ConstraintWeights_0=1) {
    alphaBoolBlend {blendOption=Linear}
    boneToModify {boneName=dyn_coat_fr_04_r}
    constraintSetup [{targetBone {boneName=thigh_r}}]
    constraintWeights [1]
   }
   n58 = "Constraint - dyn_coat_fr_mid_03_r (P)"(ComponentPose=@n57, Alpha=1, ConstraintWeights_0=1) {
    alphaBoolBlend {blendOption=Linear}
    boneToModify {boneName=dyn_coat_fr_mid_03_r}
    constraintSetup [{targetBone {boneName=thigh_r}}]
    constraintWeights [1]
   }
   n59 = "Constraint - dyn_coat_out_03_r (P)"(ComponentPose=@n58, Alpha=1, ConstraintWeights_0=1) {
    alphaBoolBlend {blendOption=Linear}
    boneToModify {boneName=dyn_coat_out_03_r}
    constraintSetup [{targetBone {boneName=thigh_r}}]
    constraintWeights [1]
   }
   n60 = Pose Driver - Source: lowerarm_r!(SourcePose:Pose Link Structure=Blend Poses by bool(BlendPose_0=Component To Local(ComponentPose=@n43), BlendPose_1=Component To Local(ComponentPose=@n59), bActiveValue=Get bRigidBodyEnabled()) {blendType=Linear}) -> (Pose:Pose Link Structure) {
    is=title AnimGraphNode_PoseDriver
    poseAsset=&/M_MED_Wizard_Arm_Poses_PoseAsset
    onlyDriveBones [{boneName=dyn_coat_sleeve_fr_r}, {boneName=dyn_coat_sleeve_in_r}, {boneName=dyn_coat_sleeve_bk_r}, {boneName=dyn_coat_sleeve_out_r}]
    poseTargets [
     {
      bApplyCustomCurve=false
      bIsHidden=false
      distanceMethod=DefaultMethod
      drivenName=base_pose
      functionType=DefaultFunction
      targetScale=0.3
      boneTransforms [{targetRotation=<rot -2.196 28.3403 4.0638>, targetTranslation=<-26.6515 0 0>}]
      customCurve {defaultValue=3.4028234663852886e+38, postInfinityExtrap=RCCE_Constant, preInfinityExtrap=RCCE_Constant, keys []}
      targetRotation=<rot 0 0 0>
     }
     {
      bApplyCustomCurve=false
      bIsHidden=false
      distanceMethod=DefaultMethod
      drivenName=Pose_5
      functionType=DefaultFunction
      targetScale=0.35
      boneTransforms [{targetRotation=<rot -0.0424 122.3283 0.0282>, targetTranslation=<-26.6322 0.0069 0.0162>}]
      customCurve {defaultValue=3.4028234663852886e+38, postInfinityExtrap=RCCE_Constant, preInfinityExtrap=RCCE_Constant, keys []}
      targetRotation=<rot 0 0 0>
     }
     {
      bApplyCustomCurve=false
      bIsHidden=false
      distanceMethod=DefaultMethod
      drivenName=Pose_6
      functionType=DefaultFunction
      targetScale=0.35
      boneTransforms [{targetRotation=<rot 0.0004 107.9911 0.0013>, targetTranslation=<-26.6513 0.0002 -0.0005>}]
      customCurve {defaultValue=3.4028234663852886e+38, postInfinityExtrap=RCCE_Constant, preInfinityExtrap=RCCE_Constant, keys []}
      targetRotation=<rot 0 0 0>
     }
    ]
    rBFParams {function=Linear, radius=80.8581}
    sourceBones [{boneName=lowerarm_r}]
   }
   n61 = Pose Driver - Source: lowerarm_l!(SourcePose:Pose Link Structure=@n60) -> (Pose:Pose Link Structure) {
    is=title AnimGraphNode_PoseDriver
    poseAsset=&M_MED_Wizard_Arm_Poses_PoseAsset
    onlyDriveBones [{boneName=dyn_coat_sleeve_out_l}, {boneName=dyn_coat_sleeve_fr_l}, {boneName=dyn_coat_sleeve_bk_l}]
    poseTargets [
     {
      bApplyCustomCurve=false
      bIsHidden=false
      distanceMethod=DefaultMethod
      drivenName=base_pose
      functionType=DefaultFunction
      targetScale=0.35
      boneTransforms [{targetRotation=<rot -2.196 28.3403 4.0638>, targetTranslation=<26.6514 0 0>}]
      customCurve {defaultValue=3.4028234663852886e+38, postInfinityExtrap=RCCE_Constant, preInfinityExtrap=RCCE_Constant, keys []}
      targetRotation=<rot 0 0 0>
     }
     {
      bApplyCustomCurve=false
      bIsHidden=false
      distanceMethod=DefaultMethod
      drivenName=Pose_5
      functionType=DefaultFunction
      targetScale=0.5992
      boneTransforms [{targetRotation=<rot -0.328 122.4869 0.1649>, targetTranslation=<26.5231 -0.0506 -0.0947>}]
      customCurve {defaultValue=3.4028234663852886e+38, postInfinityExtrap=RCCE_Constant, preInfinityExtrap=RCCE_Constant, keys []}
      targetRotation=<rot 0 0 0>
     }
    ]
    rBFParams {function=Linear, radius=92.7289}
    sourceBones [{boneName=lowerarm_l}]
   }
   n62 = Get aim_pitch_float() -> (aim_pitch_float:double)
   n63 = Flap Node - Bone: dyn_coat_sleeve_out_r!(ComponentPose:Component Space Pose Link Structure=Local To Component(LocalPose=@n61), Alpha:float=@n62) -> (Pose:Component Space Pose Link Structure) {
    is=title FortAnimGraphNode_Flap
    targetComponent=TranslationX
    alphaBoolBlend {blendOption=Linear}
    sourceDrives [{
      bUseRange=false
      drivingCurve=null
      multiplier=-4.25
      rangeMax=1
      rangeMin=-1
      remappedMax=1
      remappedMin=0
      sourceComponent=RotationZ
      useQuaternion=true
      sourceBone {boneName=lowerarm_r}
     }]
    targetBone {boneName=dyn_coat_sleeve_out_r}
   }
   n64 = Flap Node - Bone: dyn_coat_sleeve_out_r(ComponentPose=@n63, Alpha=@n62) {
    targetComponent=RotationZ
    alphaBoolBlend {blendOption=Linear}
    sourceDrives [{
      bUseRange=false
      drivingCurve=null
      multiplier=-0.2
      rangeMax=1
      rangeMin=-1
      remappedMax=1
      remappedMin=0
      sourceComponent=RotationZ
      useQuaternion=true
      sourceBone {boneName=lowerarm_r}
     }]
    targetBone {boneName=dyn_coat_sleeve_out_r}
   }
   n65 = Flap Node - Bone: dyn_coat_sleeve_out_l!(ComponentPose:Component Space Pose Link Structure=@n64, Alpha:float=@n62) -> (Pose:Component Space Pose Link Structure) {
    is=title FortAnimGraphNode_Flap
    targetComponent=TranslationX
    alphaBoolBlend {blendOption=Linear}
    sourceDrives [{
      bUseRange=false
      drivingCurve=null
      multiplier=3.5
      rangeMax=1
      rangeMin=-1
      remappedMax=1
      remappedMin=0
      sourceComponent=RotationZ
      useQuaternion=true
      sourceBone {boneName=lowerarm_r}
     }]
    targetBone {boneName=dyn_coat_sleeve_out_l}
   }
   n66 = "Transform (Modify) Bone - Bone: dyn_coat_lapel_r"(ComponentPose=@n65, Alpha=Get ADS_R_arm() -> (ADS_R_arm:double), Translation="0,0,0", Rotation="0,95,20", Scale="0.8,1,1") {
    rotationMode=BMM_Additive
    alphaBoolBlend {blendOption=Linear}
    alphaScaleBiasClamp {bInterpResult=true}
    boneToModify {boneName=dyn_coat_lapel_r}
    rotation {roll=20, yaw=95}
    scale {x=0.8}
   }
   n67 = "Transform (Modify) Bone - Bone: dyn_coat_lapel_l"(ComponentPose=@n66, Alpha=Get ADS_L_arm() -> (ADS_L_arm:double), Translation="0,0,0", Rotation="0,-95,20", Scale="1,1,1") {
    rotationMode=BMM_Additive
    alphaBoolBlend {blendOption=Linear}
    alphaScaleBiasClamp {bInterpResult=true}
    boneToModify {boneName=dyn_coat_lapel_l}
    rotation {roll=20, yaw=-95}
   }
   n68 = Get No Jump() -> (No Jump:double)
   n69 = "Constraint - dyn_coat_fr_05_l (P)"(ComponentPose=@n67, Alpha=@n68, ConstraintWeights_0=1) {
    alpha=0
    alphaBoolBlend {blendOption=Linear}
    alphaScaleBiasClamp {bClampResult=true, clampMax=0.25}
    boneToModify {boneName=dyn_coat_fr_05_l}
    constraintSetup [{targetBone {boneName=thigh_l}}]
    constraintWeights [1]
   }
   n70 = "Constraint - dyn_coat_fr_mid_03_l (P)"(ComponentPose=@n69, Alpha=@n68, ConstraintWeights_0=1) {
    alpha=0
    alphaBoolBlend {blendOption=Linear}
    alphaScaleBiasClamp {bClampResult=true, clampMax=0.25}
    boneToModify {boneName=dyn_coat_fr_mid_03_l}
    constraintSetup [{targetBone {boneName=thigh_l}}]
    constraintWeights [1]
   }
   n71 = "Constraint - dyn_coat_fr_05_r (P)"(ComponentPose=@n70, Alpha=@n68, ConstraintWeights_0=1) {
    alpha=0
    alphaBoolBlend {blendOption=Linear}
    alphaScaleBiasClamp {bClampResult=true, clampMax=0.25}
    boneToModify {boneName=dyn_coat_fr_05_r}
    constraintSetup [{targetBone {boneName=thigh_r}}]
    constraintWeights [1]
   }
   n72 = "Constraint - dyn_coat_fr_mid_03_r (P)"(ComponentPose=@n71, Alpha=@n68, ConstraintWeights_0=1) {
    alpha=0
    alphaBoolBlend {blendOption=Linear}
    alphaScaleBiasClamp {bClampResult=true, clampMax=0.25}
    boneToModify {boneName=dyn_coat_fr_mid_03_r}
    constraintSetup [{targetBone {boneName=thigh_r}}]
    constraintWeights [1]
   }
   n73 = Get Jump() -> (Jump:double)
   n74 = "Transform (Modify) Bone - Bone: dyn_coat_fr_02_r"(ComponentPose=@n72, Alpha=@n73, Translation="2.062,1.420,-3.195", Rotation="-4.612,-43.505,-40.260", Scale="1,1,1") {
    rotationMode=BMM_Additive
    rotationSpace=BCS_BoneSpace
    translationMode=BMM_Additive
    translationSpace=BCS_BoneSpace
    alphaBoolBlend {blendOption=Linear}
    alphaScaleBiasClamp {bClampResult=true, bInterpResult=true, interpSpeedDecreasing=5, interpSpeedIncreasing=5}
    boneToModify {boneName=dyn_coat_fr_02_r}
    rotation=<rot -4.612 -43.505 -40.2599>
    translation=<2.062 1.42 -3.195>
   }
   n75 = "Transform (Modify) Bone - Bone: dyn_coat_fr_mid_02_r"(ComponentPose=@n74, Alpha=@n73, Translation="2.062,1.420,-3.195", Rotation="-11.682,30,-41.263", Scale="1,1,1") {
    rotationMode=BMM_Additive
    rotationSpace=BCS_BoneSpace
    translationMode=BMM_Additive
    translationSpace=BCS_BoneSpace
    alphaBoolBlend {blendOption=Linear}
    alphaScaleBiasClamp {bClampResult=true, bInterpResult=true, interpSpeedDecreasing=5, interpSpeedIncreasing=5}
    boneToModify {boneName=dyn_coat_fr_mid_02_r}
    rotation=<rot -11.682 30 -41.263>
    translation=<2.062 1.42 -3.195>
   }
   n76 = "Transform (Modify) Bone - Bone: dyn_scarf_r_01"!(ComponentPose=@n75, Alpha=1, Rotation="5.000,0.000,0.000") -> (Pose) {
    alpha=0
    rotationMode=BMM_Additive
    rotationSpace=BCS_BoneSpace
    translationMode=BMM_Additive
    translationSpace=BCS_BoneSpace
    alphaBoolBlend {blendOption=Linear}
    alphaScaleBiasClamp {bClampResult=true, interpSpeedDecreasing=5, interpSpeedIncreasing=5}
    boneToModify {boneName=dyn_scarf_r_01}
    rotation {pitch=5}
   }
   Output Pose(Result=Component To Local(ComponentPose=@n76)) {layerGroup=None}
   n77 = "Constraint - dyn_coat_fr_02_l (PPP)"!(
    ComponentPose=Local To Component(LocalPose=Copy Pose From Mesh() {bUseAttachedParent=true})
    Alpha=1
    ConstraintWeights_0=0.4907
    ConstraintWeights_1:float=0.1011
    ConstraintWeights_2:float=0.4082
   ) -> (Pose) {
    alphaBoolBlend {blendOption=Linear}
    boneToModify {boneName=dyn_coat_fr_02_l}
    constraintSetup [{targetBone {boneName=dyn_coat_fr_02_l}}, {targetBone {boneName=dyn_coat_fr_mid_01_l}}, {targetBone {boneName=dyn_coat_fr_01_l}}]
    constraintWeights [1,1,1]
   }
   n78 = "Constraint - dyn_loincloth_01_l (PPP)"(ComponentPose=@n77, Alpha=1, ConstraintWeights_0=0.6434, ConstraintWeights_1=0.3355, ConstraintWeights_2=0.0211) {
    alphaBoolBlend {blendOption=Linear}
    boneToModify {boneName=dyn_loincloth_01_l}
    constraintSetup [{targetBone {boneName=pelvis}}, {targetBone {boneName=thigh_twist_01_l}}, {targetBone {boneName=dyn_loincloth_01_l}}]
    constraintWeights [1,1,1]
   }
   n79 = "Constraint - dyn_loincloth_02_l (PPP)"(ComponentPose=@n78, Alpha=1, ConstraintWeights_0=0.2074, ConstraintWeights_1=0.7589, ConstraintWeights_2=0.0337) {
    alphaBoolBlend {blendOption=Linear}
    boneToModify {boneName=dyn_loincloth_02_l}
    constraintSetup [{targetBone {boneName=pelvis}}, {targetBone {boneName=thigh_twist_01_l}}, {targetBone {boneName=thigh_l}}]
    constraintWeights [1,1,1]
   }
   n80 = "Transform (Modify) Bone - Bone: dyn_coat_fr_01_l"!(ComponentPose=@n79, Translation="-3.013,-4.507,1.513", Rotation="20.000,-10.000,-70.001", Scale="0.9,0.8,1") -> (Pose) {
    bAlphaBoolEnabled=false
    rotationMode=BMM_Additive
    rotationSpace=BCS_BoneSpace
    scaleMode=BMM_Additive
    translationMode=BMM_Additive
    translationSpace=BCS_BoneSpace
    alphaBoolBlend {blendInTime=0.1, blendOption=Linear, blendOutTime=0.1}
    alphaScaleBiasClamp {interpSpeedDecreasing=20, interpSpeedIncreasing=20}
    boneToModify {boneName=dyn_coat_fr_01_l}
    rotation=<rot 19.9999 -10.0005 -70.001>
    scale=<0.9 0.8>
    translation=<-3.013 -4.507 1.513>
   }
   n81 = "Transform (Modify) Bone - Bone: dyn_coat_fr_03_l"!(ComponentPose=@n80, Translation="-28.879,12.430,17.570", Scale="0.8,1,1") -> (Pose) {
    bAlphaBoolEnabled=false
    rotationMode=BMM_Additive
    rotationSpace=BCS_BoneSpace
    scaleMode=BMM_Additive
    translationMode=BMM_Additive
    translationSpace=BCS_BoneSpace
    alphaBoolBlend {blendInTime=0.1, blendOption=Linear, blendOutTime=0.1}
    alphaScaleBiasClamp {interpSpeedDecreasing=20, interpSpeedIncreasing=20}
    boneToModify {boneName=dyn_coat_fr_03_l}
    scale {x=0.8}
    translation=<-28.879 12.43 17.57>
   }
   n82 = "Transform (Modify) Bone - Bone: dyn_coat_out_01_l"!(ComponentPose=@n81, Alpha=1, Translation="-2,0.000,0") -> (Pose) {
    rotationMode=BMM_Additive
    rotationSpace=BCS_BoneSpace
    scaleMode=BMM_Additive
    translationMode=BMM_Additive
    translationSpace=BCS_BoneSpace
    alphaBoolBlend {blendOption=Linear}
    alphaScaleBiasClamp {interpSpeedDecreasing=20, interpSpeedIncreasing=20}
    boneToModify {boneName=dyn_coat_out_01_l}
    translation {x=-2}
   }
   n83 = "Transform (Modify) Bone - Bone: dyn_coat_fr_03_r"!(ComponentPose=@n82, Alpha=1, Translation="10,0.000,0") -> (Pose) {
    rotationMode=BMM_Additive
    rotationSpace=BCS_BoneSpace
    scaleMode=BMM_Additive
    translationMode=BMM_Additive
    translationSpace=BCS_BoneSpace
    alphaBoolBlend {blendOption=Linear}
    alphaScaleBiasClamp {interpSpeedDecreasing=20, interpSpeedIncreasing=20}
    boneToModify {boneName=dyn_coat_fr_03_r}
    translation {x=10}
   }
   n84 = "Transform (Modify) Bone - Bone: dyn_coat_fr_mid_01_l"!(ComponentPose=@n83, Alpha=1, Translation="-191.593,-5.027,-8.030") -> (Pose) {
    rotationMode=BMM_Additive
    rotationSpace=BCS_BoneSpace
    scaleMode=BMM_Additive
    translationMode=BMM_Additive
    translationSpace=BCS_BoneSpace
    alphaBoolBlend {blendInTime=0.1, blendOption=Linear, blendOutTime=0.1}
    alphaScaleBiasClamp {interpSpeedDecreasing=20, interpSpeedIncreasing=20}
    boneToModify {boneName=dyn_coat_fr_mid_01_l}
    translation=<-191.593 -5.027 -8.03>
   }
   n85 = "Transform (Modify) Bone - Bone: dyn_coat_fr_01_l"!(ComponentPose=@n84, Alpha:float=1, Translation="-3.5,-2,2.000", Rotation="0,0,-3") -> (Pose) {
    bAlphaBoolEnabled=false
    rotationMode=BMM_Additive
    rotationSpace=BCS_BoneSpace
    scaleMode=BMM_Additive
    translationMode=BMM_Additive
    translationSpace=BCS_BoneSpace
    alphaBoolBlend {blendInTime=0.1, blendOption=Linear, blendOutTime=0.1}
    alphaScaleBiasClamp {bInterpResult=true, interpSpeedDecreasing=15, interpSpeedIncreasing=15}
    boneToModify {boneName=dyn_coat_fr_01_l}
    rotation {roll=-3}
    translation=<-3.5 -2 2>
   }
   n86 = "Transform (Modify) Bone - Bone: dyn_coat_fr_mid_01_l"!(ComponentPose=@n85, Alpha=1, Translation="0,-5,3", Rotation:Rotator="0.000,0.000,15") -> (Pose) {
    bAlphaBoolEnabled=false
    rotationMode=BMM_Additive
    rotationSpace=BCS_BoneSpace
    scaleMode=BMM_Additive
    translationMode=BMM_Additive
    translationSpace=BCS_BoneSpace
    alphaBoolBlend {blendInTime=0.1, blendOption=Linear, blendOutTime=0.1}
    alphaScaleBiasClamp {bInterpResult=true, interpSpeedDecreasing=15, interpSpeedIncreasing=15}
    boneToModify {boneName=dyn_coat_fr_mid_01_l}
    rotation {roll=15}
    translation {y=-5, z=3}
   }
   n87 = Flap Node - Bone: dyn_pipe!(ComponentPose:Component Space Pose Link Structure=@n86, Alpha:float=1) -> (Pose:Component Space Pose Link Structure) {
    is=title FortAnimGraphNode_Flap
    bUseClamp=true
    targetClampMax=100
    targetClampMin=-100
    targetComponent=RotationZ
    alphaBoolBlend {blendOption=Linear}
    sourceDrives [{
      bUseRange=false
      drivingCurve=null
      multiplier=-0.9
      rangeMax=1
      rangeMin=-1
      remappedMax=1
      remappedMin=0
      sourceComponent=RotationZ
      useQuaternion=true
      sourceBone {boneName=thigh_r}
     }]
    targetBone {boneName=dyn_pipe}
   }
   n88 = RigidBody!(ComponentPose=@n87, Alpha, OverrideWorldGravity:Vector="0,0,-980", ExternalForce) -> (Pose) {
    bClampLinearTranslationLimitToRefPose=true
    bOverrideWorldGravity=true
    overridePhysicsAsset=&M_MED_Wizard_Dyn_Coat
    simulationSpace=BaseBoneSpace
    alphaBoolBlend {blendOption=Linear}
    baseBoneRef {boneName=spine_01}
    componentAppliedLinearAccClamp=<5000 5000 5000>
    componentLinearAccScale=<0.5 0.5 0.4>
    componentLinearVelScale=<0.5 0.5 0.4>
    overrideWorldGravity {z=-980}
   }
   Save cached pose 'Crouch Coat'(Pose=Component To Local(ComponentPose=@n88)) {cacheName=Crouch Coat}
   Save cached pose 'AnimCache'(Pose=Copy Pose From Mesh() {bUseAttachedParent=true}) {cacheName=AnimCache}
   Use cached pose 'AnimCache'()
   Flap Node - Bone: dyn_pipe(Alpha=1) {
    bUseClamp=true
    targetClampMax=100
    targetClampMin=-100
    targetComponent=RotationZ
    alphaBoolBlend {blendOption=Linear}
    sourceDrives [{
      bUseRange=false
      drivingCurve=null
      multiplier=-0.875
      rangeMax=1
      rangeMin=-1
      remappedMax=1
      remappedMin=0
      sourceComponent=RotationZ
      useQuaternion=true
      sourceBone {boneName=thigh_r}
     }]
    targetBone {boneName=dyn_pipe}
   }
   Break Fort Anim Input Skydiving External Force(FortAnimInput_SkydivingExternalForce=Get SkydivingExternalForceInput()) -> (
    bUseSkydivingVectorForce=~
    SkydivingVectorForce=~
    HeadToPelvisDirection=~
    FloatingMultiplier=~
    FloatingAdditive=~
    DivingMultiplier=~
    DivingAdditive=~
    ParachutingMultiplier=~
    ParachutingAdditive=~
   )
   Get ADS_L_arm()
   Get Forward_direct()
  ]}
 EventGraph {nodes [
   Event BlueprintUpdateAnimation()
   > n1 = Sequence!(execute) -> (then_0, then_1, then_2:Exec, then_3:Exec, then_4:Exec, then_5:Exec, then_6:Exec, then_7:Exec, then_8:Exec)
   n2 = Branch(execute=[@n1.then_0, @n26.then], Condition=Get bIsInFrontEnd())
   > Set FrontEnd Alpha(FrontEnd Alpha:double=0.8) -> (Output_Get:double)
   Set FrontEnd Alpha(execute=@n2.else, FrontEnd Alpha=1)
   n3 = BreakTransform(InTransform=GetDeltaTransformFromRefPose(self=GetOwnerSkeletalMeshComponent(), BoneName=thigh_l, BaseName=pelvis))
   Set L Thigh Rotation(execute=@n1.then_1, L Thigh Rotation:double=MapRangeClamped(Value=BreakRotator(InRot=@n3.Rotation).Yaw, InRangeA=-20, InRangeB=-100)) -> (Output_Get:double)
   n4 = MapRangeUnclamped!(
    Value:double=Get PawnMovementDirectionAzimuth() -> (PawnMovementDirectionAzimuth:float)
    InRangeA:double=30
    InRangeB:double=90
    OutRangeA:double=0
    OutRangeB:double=0.75
   ) -> (ReturnValue:double)
   Set Forward_direct(execute=@n1.then_2, Forward_direct:double=SelectFloat(A="Absolute (Float)"(A=@n4), B=0.5, bPickA=Get bIsMoving2D())) -> (Output_Get:double)
   n5 = Get PawnMovementDirectionAzimuth()
   n6 = float - float(A=MapRangeClamped(Value=@n5, InRangeA=-45, InRangeB=-90), B=MapRangeClamped(Value=@n5, InRangeA=-90, InRangeB=-135)) {is=title K2Node_CallFunction}
   n7 = Get bIsCrouching()
   n8 = Get bIsMoving2D()
   > Set Left_direct(Left_direct:double=SelectFloat(A="Absolute (Float)"(A=@n6), B=0, bPickA=AND Boolean(A=@n7, B=@n8))) -> (Output_Get:double)
   n9 = Get PawnMovementDirectionAzimuth()
   n10 = float - float(A=MapRangeClamped(Value=@n9, InRangeA=45, InRangeB=90), B=MapRangeClamped(Value=@n9, InRangeA=90, InRangeB=135))
   n11 = Get bIsCrouching()
   n12 = Get bIsMoving2D()
   > Set Right_direct(Right_direct:double=SelectFloat(A="Absolute (Float)"(A=@n10), B=0, bPickA=AND Boolean(A=@n11, B=@n12))) -> (Output_Get:double)
   > n13 = Branch(Condition=OR Boolean(A=Get bIsJumping(), B=Get bIsFalling()))
   > Set Right Jump(Right Jump:double=1) -> (Output_Get:double)
   Set Right Jump(execute=@n13.else, Right Jump=0)
   n14 = Branch(execute=@n1.then_3, Condition=Get bIsMoving2D())
   > n15 = Branch(Condition=Get bIsCrouching())
   > Set CrouchWalking(CrouchWalking:double=0) -> (Output_Get:double)
   Set CrouchWalking(execute=[@n15.else, @n14.else], CrouchWalking=1)
   n16 = Gravity Override State Function DEPRECATED(
    execute=@n1.then_5
    CharacterPartAnimInstance=Self-Reference()
    GameTimeInSeconds=GetGameTimeInSeconds()
    GravityOverrideParams=Get Gravity Override Params()
   )
   > Set GravityOverride(GravityOverride=@n16)
   n17 = Branch(execute=@n1.then_6, Condition=Get bIsCrouching())
   > Set Off_Crouch(Off_Crouch:double=0) -> (Output_Get:double)
   Set Off_Crouch(execute=@n17.else, Off_Crouch=1)
   n18 = Branch(execute=@n1.then_7, Condition=AND Boolean(A=Get bIsMoving2D(), B=Get bIsSprinting()))
   > Set Forward_motion(Forward_motion:double=1) -> (Output_Get:double)
   Set Forward_motion(execute=@n18.else, Forward_motion=0)
   n19 = Get is_skydiving() -> (is_skydiving:bool)
   n20 = Get bIsJumping()
   n21 = Get bIsFalling()
   n22 = Branch(execute=@n1.then_8, Condition=AND Boolean(A=OR Boolean(A=@n20, B=@n21), B=NOT Boolean(A=@n19)))
   > Set Jump(Jump:double=1) -> (Output_Get:double)
   > n23 = Branch(Condition=Get bIsMoving2D())
   > Set No Jump(No Jump:double=0) -> (Output_Get:double)
   Set No Jump(execute=@n23.else, No Jump=1)
   Set Jump(execute=@n22.else, Jump=0)
   > Set No Jump(No Jump=1)
   n24 = Branch(execute=@n1.then_4, Condition=Get bIsTargeting())
   n25 = BreakTransform(InTransform=GetDeltaTransformFromRefPose(self=GetOwnerSkeletalMeshComponent(), BoneName=upperarm_l, BaseName=spine_05))
   > Set ADS_L_arm(ADS_L_arm:double=MapRangeClamped(Value=BreakRotator(InRot=@n25.Rotation).Yaw, InRangeA=36, InRangeB=42)) -> (Output_Get:double)
   > Set ADS_R_arm(ADS_R_arm:double=1) -> (Output_Get:double)
   Set ADS_L_arm(execute=@n24.else, ADS_L_arm=0)
   > Set ADS_R_arm(ADS_R_arm=0)
   Event OnParachuteOpened()
   > Set is_parachuting(is_parachuting:bool=true) -> (Output_Get:bool)
   Event OnBeginSkydiving()
   > Set is_skydiving(is_skydiving:bool=true) -> (Output_Get:bool)
   Event OnEndSkydiving()
   > Set is_skydiving(is_skydiving=false)
   Event OnEndCrouching()
   > Set is_crouching(is_crouching:bool=false) -> (Output_Get:bool)
   Event OnBeginCrouching()
   > Set is_crouching(is_crouching=true)
   Event OnParachuteClosed()
   > Set is_parachuting(is_parachuting=false)
   n26 = Set aim_pitch_float(aim_pitch_float:double=MapRangeClamped(Value=Get AimPitch(), InRangeA=0, InRangeB=90, OutRangeA=1, OutRangeB=0.35)) -> (Output_Get:double)
   n27 = GetDeltaTransformFromRefPose(self=GetOwnerSkeletalMeshComponent(), BoneName=upperarm_r, BaseName=spine_05)
   MapRangeClamped(Value=BreakRotator(InRot=BreakTransform(InTransform=@n27).Rotation).Yaw, InRangeA=36, InRangeB=92)
   AND Boolean!(A=false, B=Get bIsJumping(), C:bool=Get bIsFalling()) -> (ReturnValue)
  ]}
]
variables [
 ADS_L_arm {type=number}
 ADS_R_arm {type=number}
 aim_pitch_float {type=number}
 CrouchWalking {type=number}
 Forward_direct {type=number}
 Forward_motion {type=number}
 FrontEnd Alpha {type=number}
 Gravity Override Params {type=GravityOverrideParamsStruct}
 GravityOverride {type=Vector}
 is_crouching {type=boolean}
 is_parachuting {type=boolean}
 is_skydiving {type=boolean}
 Jump {type=number}
 L Thigh Rotation {type=number}
 Left_direct {type=number}
 No Jump {type=number}
 Off_Crouch {type=number}
 Right Jump {type=number}
 Right_direct {type=number}
]
